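; Controller requirement for NPC conversation services that passes when the requesting player's team has at least one dead member.
; IsRequirementMetInternal resolves the Fort pawn from the request context, casts its player state to FortPlayerStateAthena and tests whether the team dead count is above zero.
; Two helpers, one fetching an attribute set with a validity flag and one testing for health below zero, are defined but not called by that check.
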(fn GetControlledPawnAttributeSet
  (params (Fort%20Player%20State :type Fort%20Player%20State%20Object%20Reference) (AttributeSetClass :type Attribute%20Set%20Class%20Reference))
  (bind _returnvalue (GameplayAttributes|GetAttributeSet (Variables|/Script/FortniteGame.FortPlayerState|GetAbilitySystemComponent Fort%20Player%20State) AttributeSetClass))
  (Utilities|IsValid _returnvalue
    (:Is%20Valid
      (LocalVariables|SetisValidAttributeSet true))
    (:Is%20Not%20Valid
      (LocalVariables|SetisValidAttributeSet false)))
  (return _returnvalue (LocalVariables|GetisValidAttributeSet)))

(fn IsPawnEliminated
  (params (AttributeSet :type Object%20Reference))
  (bind _asfort_health_set (Utilities|Casting|CastToFortHealthSet AttributeSet)
    (:then
      (return (< (.CurrentValue (Utilities|Struct|BreakFortGameplayAttributeData (Variables|/Script/FortniteGame.FortHealthSet|GetHealth _asfort_health_set))) 0)))
    (:CastFailed
      (return false))))

(fn IsRequirementMetInternal
  (params (RequestContext :type Controller%20Requirement%20Test%20Context%20Structure%20%28by%20ref%29))
  (bind _fortpawn (Class|FortQuestBlueprintLibrary|GetFortPawnFromObject (Utilities|Struct|BreakControllerRequirementTestContext RequestContext))
    (:Valid
      (bind _asfort_player_state_athena (Utilities|Casting|CastToFortPlayerStateAthena (Variables|/Script/Engine.Pawn|GetPlayerState _fortpawn))
        (:then
          (bind _returnvalue (Class|FortPlayerStateAthena|GetTeamDeadCount _asfort_player_state_athena))
          (return (> _returnvalue 0)))
        (:CastFailed
          (return false))))))
