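; Controller requirement for NPC conversation services that checks whether the player has enough health to cover a service's health cost.
; It gets the ability system component from the request context and reads the health attribute from it.
; The cost is evaluated from a curve table row handle variable, so it is data-driven.
; The requirement passes only when health is strictly greater than that cost, so health equal to the cost does not qualify.
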(fn IsRequirementMetInternal
  (params (RequestContext :type Controller%20Requirement%20Test%20Context%20Structure%20%28by%20ref%29))
  (bind (_curvetable _rowname) (Utilities|Struct|BreakCurveTableRowHandle (Variables|Default|GethealthCostTable)))
  (bind _health_attribute (LocalVariables|Gethealth_attribute))
  (bind _output_get (LocalVariables|SetplayerAbilitySystem (Ability|GetAbilitySystemComponent (Utilities|Struct|BreakControllerRequirementTestContext RequestContext))))
  (if (GameplayAttributes|GetGameplayAttributeValue _output_get _health_attribute)
    (bind _outxy (DataTable|EvaluateCurveTableRow _curvetable _rowname)
      (:RowFound
        (if (> (.ReturnValue (GameplayAttributes|GetGameplayAttributeValue _output_get _health_attribute)) _outxy)
          (return true)
          (else
            _)))
      (:RowNotFound
        _))
    (return false)
    (else
      _)))
