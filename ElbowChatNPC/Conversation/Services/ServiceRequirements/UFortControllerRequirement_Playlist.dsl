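; Controller requirement for ElbowChat NPC conversation services that gates a service by the current playlist.
; It takes the test subject controller from the request context and gets the Athena game state from it.
; The requirement is met when the playlist's context tags share any exact match with the configured playlistContextTags.
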(fn IsRequirementMetInternal
  (params (RequestContext :type Controller%20Requirement%20Test%20Context%20Structure%20%28by%20ref%29))
  (bind (_testsubjectactor _testsubjectcontroller) (Utilities|Struct|BreakControllerRequirementTestContext RequestContext))
  (return (GameplayTags|HasAnyTags (Class|FortGameStateAthena|GetAthenaPlaylistContextTags (Fortnite|Athena|GameState|GetGameStateAthenafromContext _testsubjectcontroller)) (Variables|Default|GetplaylistContextTags) true)))
